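; Building game actor (BGA) for the ElbowChat NPC's shield, handling its spawn visuals and looping force-field sound.
; BeginPlay stamps the current game time into the core mesh material's StartGameTime parameter, spawns a Niagara spawn effect at the actor's location and fades in the loop.
; On death, the loop fades out after a validity check.
; The overlap and tick events are empty, and construction only calls the parent.
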(event EventBeginPlay
  (Rendering|Material|SetScalarParameterValueonMaterials (Variables|Default|GetCoreMesh) "StartGameTime" (Utilities|Time|GetGameTimeinSeconds))
  (Niagara|SpawnSystematLocation "/ElbowChatNPC/Assets/NS_NPC_ElbowChat_SpawnShield.NS_NPC_ElbowChat_SpawnShield" (Transformation|GetActorLocation self))
  (Audio|Components|Audio|FadeIn (Variables|Default|GetSpot_Magic_ForceField_Loop_Cue) 0.5 1.0 0.0 "Linear" :!class /Script/Engine.AudioComponent) ; Play force field sound
)

(event Collision|EventActorBeginOverlap
  (params (OtherActor :type Actor%20Object%20Reference)))

(event EventTick
  (params (DeltaSeconds :type Float%20%28single-precision%29)))

(event EventOnDeathPlayEffects
  (params (Damage :type Float%20%28single-precision%29) (DamageTags :type Gameplay%20Tag%20Container%20Structure%20%28by%20ref%29) (Momentum :type Vector :default "0, 0, 0") (HitInfo :type Hit%20Result%20Structure%20%28by%20ref%29) (InstigatedBy :type Fort%20Pawn%20Object%20Reference) (DamageCauser :type Actor%20Object%20Reference) (EffectContext :type Gameplay%20Effect%20Context%20Handle%20Structure))
  (bind _spot_magic_forcefield_loop_cue (Variables|Default|GetSpot_Magic_ForceField_Loop_Cue))
  ; On Death Stop Looping Force Field Sound
  (if (Utilities|IsValid _spot_magic_forcefield_loop_cue :!class /Script/Engine.KismetSystemLibrary)
    (Utilities|IsValid _spot_magic_forcefield_loop_cue
      (:Is%20Valid
        (Audio|Components|Audio|FadeOut _spot_magic_forcefield_loop_cue 0.2 0.0 "Linear" :!class /Script/Engine.AudioComponent)))))

(fn ConstructionScript
  (|CallParentFunction))
