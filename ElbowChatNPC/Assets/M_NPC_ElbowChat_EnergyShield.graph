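# Translucent, unlit, two-sided energy shield material for the ElbowChat NPC, usable on Niagara mesh particles.
# Distorted panning noise drives three-band color, tinted by time of day, and opacity built from fresnel, depth fade, vertical fades and timed dissolve.
# Timestamped hit parameters drive a world-space shockwave with tint, flash, opacity burst and vertex displacement.
# Ambient noise wobble, scaled by a particle dynamic parameter, is disabled on lower quality levels.
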Material /ElbowChatNPC/Assets/M_NPC_ElbowChat_EnergyShield
graph {
 properties {
  bUsedWithNiagaraMeshParticles=true
  blendMode=BLEND_Translucent
  pixelDepthOffsetMode=PDOM_Legacy
  shadingModel=MSM_Unlit
  twoSided=true
  usedShadingModels=Unlit
 }
 nodes [
  Constant() {r=5}
  n1 = VectorParameter() {parameterName=MorningColor, defaultValue=<rgba 0.4287 0.0782 0.0782 1>}
  n2 = VectorParameter() {parameterName=DayColor, defaultValue=<rgba 1 1 1 0>}
  n3 = VectorParameter() {parameterName=EveningColor, defaultValue=<rgba 0.4287 0.0782 0.0782 1>}
  n4 = VectorParameter() {parameterName=NightColor, defaultValue=<rgba 0.2214 0.2966 0.4167 1>}
  n5 = VectorParameter() {group=02-Color, parameterName=DarkColor, defaultValue=<rgba 0.0128 0.6823 0.2668 1>}
  n6 = ScalarParameter() {defaultValue=0.7, group=02-Color, parameterName=LightColorThreshold}
  n7 = ScalarParameter() {defaultValue=0.1, group=02-Color, parameterName=LightColorSoftness}
  n8 = VectorParameter() {group=02-Color, parameterName=ColorNoise_UVSettings, defaultValue=<rgba 1 1 0.05 0.15>}
  n9 = Multiply(
   A=TextureCoordinate()
   B=MaterialFunctionCall(X=@n8.R, Y=@n8.R) {materialFunction=&MakeFloat2}.Result
  )
  n10 = Multiply(A=MaterialFunctionCall(X=@n8.B, Y=@n8.B) {materialFunction=&MakeFloat2}.Result, B=Time())
  n11 = TextureCoordinate()
  n12 = VectorParameter() {group=03-Distortion, parameterName=Distortion_UVSettings, defaultValue=<rgba 2 0.6 -0.02 -0.05>}
  n13 = MaterialFunctionCall(X=@n12.R, Y=@n12.R) {materialFunction=&MakeFloat2}
  n14 = MaterialFunctionCall(X=@n12.B, Y=@n12.B) {materialFunction=&MakeFloat2}
  n15 = Time()
  n16 = TextureSampleParameter2D(UVs=Add(A=Multiply(A=@n11, B=@n13.Result), B=Multiply(A=@n14.Result, B=@n15))) {
   group=03-Distortion
   parameterName=DistortionTex
   samplerType=SAMPLERTYPE_LinearColor
   texture=/Game/Effects/Fort_Effects/Effects/Environment/Desert/Dust/Organize/T_Turbulence_Seamless
  }
  n17 = Multiply(
   A=ComponentMask(None=@n16.RGB) {g=true, r=true}
   B=ScalarParameter() {defaultValue=0.2, group=03-Distortion, parameterName=DistortionStrength}
  )
  n18 = TextureSampleParameter2D(UVs=Add(A=@n17, B=Add(A=@n9, B=@n10))) {
   group=02-Color
   parameterName=ColorTex
   samplerType=SAMPLERTYPE_LinearGrayscale
   texture=/Game/Weapons/FORT_Melee/Pickaxe_Cube_Ninja_Male/FX/T_FluidFlameTile
  }
  n19 = LinearInterpolate(
   A=VectorParameter() {group=02-Color, parameterName=MediumColor, defaultValue=<rgba 0.0128 0.6823 0.2668 1>}.RGB
   B=VectorParameter() {group=02-Color, parameterName=LightColor, defaultValue=<rgba 0.0128 0.6823 0.2668 1>}.RGB
   Alpha=SmoothStep(Min=@n6, Max=Add(A=@n6, B=@n7), Value=@n18.RGB)
  )
  n20 = ScalarParameter() {defaultValue=0.3, group=02-Color, parameterName=DarkColorThreshold}
  n21 = SmoothStep(Min=@n20, Max=Add(A=@n20, B=ScalarParameter() {defaultValue=0.1, group=02-Color, parameterName=DarkColorSoftness}), Value=@n18.RGB)
  n22 = ScalarParameter() {defaultValue=0.45, parameterName=Hitwave_FlashTiming}
  n23 = NamedRerouteUsage()
  n24 = OneMinus(None=SmoothStep(Min=@n22, Value=@n23) {constMin=0.6})
  n25 = Add(A=@n22) {constB=0.1}
  n26 = Multiply(
   A=Multiply(A=Power(Base=@n24) {constExponent=3}, B=SmoothStep(Min=@n22, Max=@n25, Value=@n23) {constMax=0.7, constMin=0.6})
   B=VectorParameter() {group=HitWave, parameterName=Hitwave-AddBrighness, defaultValue=<rgba 0.3311 0.6591 0.9271 0>}.RGB
  )
  n27 = LinearInterpolate(
   A=Multiply(A=LinearInterpolate(A=@n5.RGB, B=@n19, Alpha=@n21), B=Add(A=@n26))
   B=VectorParameter() {group=HitWave, parameterName=HitWave-Tint, defaultValue=<rgba 0.2236 2 0.9691 1>}.RGB
   Alpha=NamedRerouteUsage()
  )
  n28 = ScalarParameter() {defaultValue=1, group=02-Color, parameterName=Brightness}
  MaterialFunctionCall(Emissive=Multiply(
    A=MaterialFunctionCall(Color_Morning=@n1.RGB, Color_Day=@n2.RGB, Color_Evening=@n3.RGB, Color_Night=@n4.RGB) {materialFunction=&Color_Time_of_Day}.Result
    B=Multiply(A=@n27, B=@n28)
   )) {materialFunction=&MF_FX_EmissiveAdjustments}
  n29 = ScalarParameter() {group=01-Opacity, parameterName=TopFadeThreshold, sortPriority=64}
  n30 = ScalarParameter() {defaultValue=0.1, group=01-Opacity, parameterName=TopBotFadeSoftness, sortPriority=65}
  n31 = Add(A=@n29, B=@n30)
  n32 = MaterialFunctionCall() {materialFunction=&LinearGradient}
  n33 = ScalarParameter() {defaultValue=1, group=01-Opacity, parameterName=BotFadeThreshold, sortPriority=64}
  n34 = Subtract(A=@n33, B=@n30)
  n35 = LinearInterpolate(Alpha=NamedRerouteUsage()) {constA=1, constB=0.5}
  n36 = VectorParameter() {group=01-Opacity, parameterName=OpacityNoise_UVSettings, defaultValue=<rgba 1 1 0 0.2>}
  n37 = Multiply(
   A=TextureCoordinate()
   B=MaterialFunctionCall(X=@n36.R, Y=@n36.R) {materialFunction=&MakeFloat2}.Result
  )
  n38 = Multiply(
   A=MaterialFunctionCall(X=@n36.B, Y=@n36.B) {materialFunction=&MakeFloat2}.Result
   B=Time()
  )
  n39 = TextureSampleParameter2D(UVs=Add(A=Add(A=@n37, B=@n38), B=@n17)) {
   group=01-Opacity
   parameterName=OpacityTex
   samplerType=SAMPLERTYPE_LinearGrayscale
   texture=/Game/Effects/Fort_Effects/Emotes/UltraLight/Texture/T_DirectionalNoise_Greyscale
  }
  n40 = LinearInterpolate(A=Multiply(A=Constant() {r=1}, B=@n35), B=Multiply(A=Constant() {r=2}, B=@n35), Alpha=@n39.R) {constA=1, constB=2}
  n41 = Multiply(
   A=MaterialFunctionCall("FadeSensitivity (optional)"=LinearInterpolate(Alpha=@n39.R) {constA=4, constB=15}) {materialFunction=&FadeBasedOnViewAngle}.Result
   B=LinearInterpolate(Alpha=@n39.R) {constA=20, constB=0}
  )
  n42 = SmoothStep(
   Min=ScalarParameter() {group=01-Opacity, parameterName=RemapOpacityMin}
   Max=ScalarParameter() {defaultValue=1, group=01-Opacity, parameterName=RemapOpacityMax}
   Value=Multiply(A=Fresnel(ExponentIn=@n40) {baseReflectFraction=0, exponent=1}, B=Saturate(None=@n41))
  )
  n43 = NamedRerouteUsage()
  n44 = DepthFade(FadeDistance=ScalarParameter() {defaultValue=50, group=01-Opacity, parameterName=Depthfade})
  n45 = Constant() {r=1}
  n46 = ScalarParameter() {defaultValue=0.5, group=05- Dissolve, parameterName=DissolveSoftness}
  n47 = Divide(
   A=Subtract(A=Time(), B=ScalarParameter() {defaultValue=1, desc=Time-stamp, group=05- Dissolve, parameterName=StartGameTime})
   B=ScalarParameter() {defaultValue=1, group=05- Dissolve, parameterName=StartErosionDuration}
  )
  n48 = Max(A=ScalarParameter() {group=01-Opacity, parameterName=Erode}, B=OneMinus(None=Saturate(None=@n47)))
  n49 = Subtract(A=@n48, B=Multiply(A=@n46, B=OneMinus(None=@n48)))
  n50 = Add(A=@n48, B=Multiply(A=@n46, B=@n48))
  n51 = Multiply(
   A=Multiply(A=Add(A=@n42, B=@n43), B=ShadingPathSwitch(Default=@n44, Deferred=@n44, Forward=@n45, Mobile=@n45))
   B=Saturate(None=SmoothStep(Min=@n49, Max=@n50, Value=@n18.RGB))
  )
  Multiply(
   A=Min(A=SmoothStep(Min=@n29, Max=@n31, Value=@n32.VGradient), B=SmoothStep(Min=@n33, Max=@n34, Value=@n32.VGradient) {constMax=0})
   B=Multiply(A=Saturate(None=@n51), B=ScalarParameter() {defaultValue=1.5, group=01-Opacity, parameterName=OverallOpacity})
  )
  n52 = Distance(
   A=WorldPosition() {worldPositionShaderOffset=WPT_ExcludeAllShaderOffsets}.XYZ
   B=ComponentMask(None=VectorParameter() {group=HitWave, parameterName=HitWave-Loc, defaultValue=<rgba 200 0 0 0>}.RGB) {b=true, g=true, r=true}
  )
  n53 = NamedRerouteUsage()
  n54 = Multiply(
   A=ConstantBiasScale(None=@n39.R) {bias=-0.5, scale=2}
   B=Multiply(A=ScalarParameter() {group=HitWave, parameterName=HitWaveRingDistortion}, B=@n53)
  )
  n55 = ScalarParameter() {group=HitWave, parameterName=HitWave-OuterRingSize}
  n56 = ScalarParameter() {group=HitWave, parameterName=HitWaveRingSoftness}
  n57 = Divide(A=Subtract(A=Add(A=@n52, B=@n54), B=Multiply(A=@n55, B=@n53)), B=Add(A=Multiply(A=@n56, B=@n53)))
  n58 = Multiply(A=OneMinus(None=Saturate(None=@n57)), B=NamedRerouteUsage())
  n59 = Constant()
  NamedRerouteDeclaration(None=ShadingPathSwitch(Default=@n58, Deferred=@n58, Forward=@n59, Mobile=@n59)) {name=ImpactShockwaveInfo, nodeColor=<rgba 0.2187 1 0 1>}
  n60 = Subtract(A=Time(), B=ScalarParameter() {defaultValue=1, desc=Time-stamp, group=HitWave, parameterName=GameTime})
  n61 = ScalarParameter() {defaultValue=2, group=HitWave, parameterName=HitWave-Duration}
  NamedRerouteDeclaration(None=Saturate(None=Divide(A=@n60, B=@n61))) {name="Impactinfo(0-1)", nodeColor=<rgba 0.9922 1 0 1>}
  n62 = ScalarParameter() {defaultValue=0.045, parameterName=Hitwave_OpacityBurstTiming}
  n63 = NamedRerouteUsage()
  n64 = OneMinus(None=SmoothStep(Min=@n62, Value=@n63) {constMin=0.6})
  n65 = Add(A=@n62) {constB=0}
  NamedRerouteDeclaration(None=Multiply(A=Power(Base=@n64) {constExponent=1}, B=SmoothStep(Min=@n62, Max=@n65, Value=@n63) {constMax=0.7, constMin=0.6})) {name=OpacityBurst, nodeColor=<rgba 1 0.838 0 1>}
  ReflectionVectorWS(CustomWorldNormal=@n16.RGB)
  n66 = Power(Base=Saturate(None=NamedRerouteUsage()), Exp=ScalarParameter() {defaultValue=1, group=HitWave, parameterName=Hitwave-pow})
  n67 = Multiply(A=Multiply(A=Saturate(None=@n66), B=ScalarParameter() {defaultValue=45, group=HitWave, parameterName=WPO-AddWave2}), B=VertexNormalWS())
  n68 = Multiply(
   A=SphereMask(A=WorldPosition().XYZ, B=VectorParameter() {group=HitWave, parameterName=HitWave-Loc, defaultValue=<rgba 200 0 0 0>}.RGB) {hardnessPercent=0}
   B=ScalarParameter() {group=04-WPO, parameterName=Wpo_ImpactAdd}
  )
  n69 = Subtract(A=Time(), B=ScalarParameter() {defaultValue=1, desc=Time-stamp, group=HitWave, parameterName=GameTime})
  n70 = Sine(None=Saturate(None=Divide(A=@n69) {constB=0.2}))
  n71 = Transform(None=Multiply(A=VectorParameter() {group=HitWave, parameterName=HitWave-Normal}.RGB, B=Multiply(A=@n68, B=@n70)) {constB=100}) {transformSourceType=TRANSFORMSOURCE_Local}
  n72 = Constant()
  n73 = ComponentMask(None=TextureCoordinate()) {g=true}
  n74 = Saturate(None=Sine(None=Multiply(A=@n73) {constB=0.5}))
  n75 = TextureCoordinate()
  n76 = VectorParameter() {group=04-WPO, parameterName=WPONoise_UVSettings, defaultValue=<rgba 1 1.2 0.07 -0.25>}
  n77 = MaterialFunctionCall(X=@n76.R, Y=@n76.R) {materialFunction=&MakeFloat2}
  n78 = MaterialFunctionCall(X=@n76.B, Y=@n76.B) {materialFunction=&MakeFloat2}
  n79 = Time()
  n80 = TextureSample(UVs=Add(A=Multiply(A=@n75, B=@n77.Result), B=Multiply(A=@n78.Result, B=@n79))) {samplerType=SAMPLERTYPE_Grayscale, texture=&LowResBlurredNoise}
  n81 = Multiply(
   A=Multiply(A=VertexNormalWS(), B=Multiply(A=@n74, B=@n80.RGB))
   B=Multiply(A=ScalarParameter() {defaultValue=40, group=04-WPO, parameterName=WPO}, B=DynamicParameter() {paramNames [WPO,Param2,Param3,Param4]}.WPO)
  )
  Reroute(None=Add(
    A=Add(A=@n67, B=@n71) {constB=0}
    B=MaterialFunctionCall!(
     Low _ PC Low and Med=@n72
     "High _ XB1, PS4"=@n81
     "Epic _ PC High, XSX, PS5"=@n81
     --------
     "Mobile Low _ PC Perf, Android"=@n72
     Mobile Med _ Switch=@n72
    ) {materialFunction=/Game/Global/MaterialFunctions/TechArt/MF_Fortnite_QualitySwitch_Scalar}.Result
   ))
  n82 = Time()
  SmoothStep(Value=Frac(None=Multiply(A=@n82) {constB=0.4})) {constMax=0.5}
 ]
}
